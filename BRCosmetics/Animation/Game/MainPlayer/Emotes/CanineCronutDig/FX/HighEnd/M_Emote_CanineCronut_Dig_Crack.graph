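# Translucent lit Niagara sprite and mesh material for the glowing ground crack in the CanineCronut Dig emote.
# Parallax occlusion mapping offsets UVs for particle-tinted color and the lava glow, which is divided by pre-exposure.
# A dynamic parameter drives a flippable smoothstep erosion of the opacity.
# Distance-field queries build a downward surface-hugging offset that fades over air gaps, gated by FRONT_END_WPO_MOD and zero on mobile feature levels.
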Material /BRCosmetics/Animation/Game/MainPlayer/Emotes/CanineCronutDig/FX/HighEnd/M_Emote_CanineCronut_Dig_Crack
graph {
 expressions {
  MaterialExpressionAbs_7 {class=MaterialExpressionAbs, inputs [None {from=MaterialExpressionMax_0, output=""}], properties {}}
  MaterialExpressionAdd_0 {
   class=MaterialExpressionAdd
   inputs [
    A {from=MaterialExpressionMultiply_0}
    B {from=MaterialExpressionAppendVector_2}
   ]
  }
  MaterialExpressionAdd_1 {
   class=MaterialExpressionAdd
   inputs [
    A {from=MaterialExpressionScalarParameter_5}
    B {from=MaterialExpressionDynamicParameter_0, output=ErosionProgress}
   ]
  }
  MaterialExpressionAdd_2 {
   class=MaterialExpressionAdd
   inputs [
    A {from=MaterialExpressionMaterialFunctionCall_8, output=Result}
    B {from=MaterialExpressionScalarParameter_6}
   ]
  }
  MaterialExpressionAdd_24 {
   class=MaterialExpressionAdd
   inputs [
    A {from=MaterialExpressionReroute_71}
    B {from=MaterialExpressionReroute_72}
   ]
  }
  MaterialExpressionAdd_25 {
   class=MaterialExpressionAdd
   inputs [
    A {from=MaterialExpressionScalarParameter_27}
    B {from=MaterialExpressionScalarParameter_28}
   ]
  }
  MaterialExpressionAppendVector_1 {
   class=MaterialExpressionAppendVector
   inputs [
    A {from=MaterialExpressionConstant_1, output=""}
    B {from=MaterialExpressionConstant_2, output=""}
   ]
  }
  MaterialExpressionAppendVector_2 {
   class=MaterialExpressionAppendVector
   inputs [
    A {from=MaterialExpressionConstant_3, output=""}
    B {from=MaterialExpressionConstant_4, output=""}
   ]
  }
  MaterialExpressionClamp_6 {
   class=MaterialExpressionClamp
   inputs [
    None {from=MaterialExpressionReroute_69, output=""}
    Min {from=MaterialExpressionOneMinus_9, output=""}
    Max {from=MaterialExpressionScalarParameter_26, output=""}
   ]
   properties {}
  }
  MaterialExpressionConstant4Vector_1 {class=MaterialExpressionConstant4Vector, inputs [], properties {constant=<rgba 1 0 0 0>}}
  MaterialExpressionConstant_0 {class=MaterialExpressionConstant}
  MaterialExpressionConstant_1 {class=MaterialExpressionConstant, properties {r=1}}
  MaterialExpressionConstant_10 {class=MaterialExpressionConstant, properties {r=1}}
  MaterialExpressionConstant_12 {class=MaterialExpressionConstant, properties {r=1}}
  MaterialExpressionConstant_2 {class=MaterialExpressionConstant, properties {r=1}}
  MaterialExpressionConstant_3 {class=MaterialExpressionConstant}
  MaterialExpressionConstant_4 {class=MaterialExpressionConstant}
  MaterialExpressionConstant_5 {class=MaterialExpressionConstant, properties {r=1}}
  MaterialExpressionConstant_6 {class=MaterialExpressionConstant}
  MaterialExpressionConstant_9 {class=MaterialExpressionConstant}
  MaterialExpressionConvert_0 {
   class=MaterialExpressionConvert
   inputs [
    None {from=""}
    None {from=""}
    None {from=MaterialExpressionScalarParameter_9}
   ]
   properties {
    nodeName=Make Vector3
    convertInputs [
     {type=Scalar, expressionInput {expression=null, inputName=None, mask=0, maskA=0, maskB=0, maskG=0, maskR=0, outputIndex=0}}
     {type=Scalar, expressionInput {expression=null, inputName=None, mask=0, maskA=0, maskB=0, maskG=0, maskR=0, outputIndex=0}}
     {
      type=Scalar
      expressionInput {
       expression=&self:MaterialExpressionScalarParameter_9
       inputName=None
       mask=0
       maskA=0
       maskB=0
       maskG=0
       maskR=0
       outputIndex=0
      }
     }
    ]
    convertOutputs [{type=Vector3}]
   }
  }
  MaterialExpressionConvert_10 {
   class=MaterialExpressionConvert
   inputs [
    None {from=""}
    None {from=""}
    None {from=MaterialExpressionScalarParameter_25}
   ]
   properties {
    nodeName=Make Vector3
    convertInputs [
     {type=Scalar, expressionInput {expression=null, inputName=None, mask=0, maskA=0, maskB=0, maskG=0, maskR=0, outputIndex=0}}
     {type=Scalar, expressionInput {expression=null, inputName=None, mask=0, maskA=0, maskB=0, maskG=0, maskR=0, outputIndex=0}}
     {
      type=Scalar
      expressionInput {
       expression=&self:MaterialExpressionScalarParameter_25
       inputName=None
       mask=0
       maskA=0
       maskB=0
       maskG=0
       maskR=0
       outputIndex=0
      }
     }
    ]
    convertOutputs [{type=Vector3}]
   }
  }
  MaterialExpressionConvert_9 {
   class=MaterialExpressionConvert
   inputs [
    None {from=""}
    None {from=""}
    None {from=MaterialExpressionReroute_70}
   ]
   properties {
    nodeName=Make Vector3
    convertInputs [
     {type=Scalar, expressionInput {expression=null, inputName=None, mask=0, maskA=0, maskB=0, maskG=0, maskR=0, outputIndex=0}}
     {type=Scalar, expressionInput {expression=null, inputName=None, mask=0, maskA=0, maskB=0, maskG=0, maskR=0, outputIndex=0}}
     {
      type=Scalar
      expressionInput {
       expression=&self:MaterialExpressionReroute_70
       inputName=None
       mask=0
       maskA=0
       maskB=0
       maskG=0
       maskR=0
       outputIndex=0
      }
     }
    ]
    convertOutputs [{type=Vector3}]
   }
  }
  MaterialExpressionDistanceToNearestSurface_0 {class=MaterialExpressionDistanceToNearestSurface, inputs [World Position {from=MaterialExpressionReroute_0, output=""}], properties {}}
  MaterialExpressionDistanceToNearestSurface_4 {class=MaterialExpressionDistanceToNearestSurface, inputs [World Position {from=MaterialExpressionReroute_64, output=""}], properties {}}
  MaterialExpressionDivide_0 {
   class=MaterialExpressionDivide
   inputs [
    A {from=MaterialExpressionMultiply_3}
    B {from=MaterialExpressionViewProperty_0, output=Property}
   ]
  }
  MaterialExpressionDynamicParameter_0 {
   class=MaterialExpressionDynamicParameter
   inputs []
   properties {defaultValue=<rgba 0 0 1 1>, paramNames [ErosionProgress, FlipErosion, Param3, Param4]}
  }
  MaterialExpressionFeatureLevelSwitch_0 {
   class=MaterialExpressionFeatureLevelSwitch
   inputs [
    Default {from=MaterialExpressionIf_1, output=""}
    ES2 {from=MaterialExpressionConstant_0, output=""}
    ES3_1 {from=MaterialExpressionConstant_0, output=""}
    SM4_REMOVED {from="", output=""}
    SM5 {from=MaterialExpressionIf_1, output=""}
    SM6 {from=MaterialExpressionIf_1, output=""}
   ]
   properties {}
  }
  MaterialExpressionIf_0 {
   class=MaterialExpressionIf
   inputs [
    A {from=MaterialExpressionDynamicParameter_0, output=FlipErosion}
    B {from="", output=""}
    "A > B" {from=MaterialExpressionOneMinus_1, output=""}
    "A == B" {from=MaterialExpressionTextureSampleParameter2D_3, output=R}
    "A < B" {from=MaterialExpressionTextureSampleParameter2D_3, output=R}
   ]
   properties {}
  }
  MaterialExpressionIf_1 {
   class=MaterialExpressionIf
   inputs [
    A {from=MaterialExpressionScalarParameter_7, output=""}
    B {from="", output=""}
    "A > B" {from=MaterialExpressionMultiply_29, output=""}
    "A == B" {from=MaterialExpressionMultiply_29, output=""}
    "A < B" {from=MaterialExpressionConstant_6, output=""}
   ]
   properties {constB=1}
  }
  MaterialExpressionMaterialFunctionCall_2 {
   class=MaterialExpressionMaterialFunctionCall
   inputs [
    CameraVec {from=""}
    Heightmap Texture {from=MaterialExpressionTextureObjectParameter_0}
    Height Ratio {from=MaterialExpressionScalarParameter_0}
    Min Steps {from=MaterialExpressionScalarParameter_1}
    Max Steps {from=MaterialExpressionScalarParameter_2}
    UVs {from=MaterialExpressionMultiply_1}
    Heightmap Channel {from=MaterialExpressionConstant4Vector_1, output=RGBA}
    Reference Plane {from=MaterialExpressionScalarParameter_3}
    ----------------------------- {from=""}
    Use World Coordinates {from=""}
    Specify Manual Texture Size {from=""}
    Manual Texture Size {from=""}
    ------------------------------ {from=""}
    "Render Shadows (Occlusion Mapping)" {from=""}
    Light Vector {from=""}
    Shadow Steps {from=""}
    Shadow Penumbra {from=""}
    Transform To VertexNormal {from=""}
   ]
   properties {materialFunction=&/MF_Emote_CanineCronut_Dig_ParallaxOcclusionMapping}
  }
  MaterialExpressionMaterialFunctionCall_8 {
   class=MaterialExpressionMaterialFunctionCall
   inputs [
    Input {from=MaterialExpressionAdd_1}
    Input Low {from=MaterialExpressionConstant_9}
    Input High {from=MaterialExpressionConstant_10}
    Target Low {from=MaterialExpressionSubtract_0}
    Target High {from=MaterialExpressionConstant_12}
   ]
   properties {materialFunction=&RemapValueRange}
  }
  MaterialExpressionMax_0 {
   class=MaterialExpressionMax
   inputs [
    A {from=MaterialExpressionReroute_65}
    B {from=MaterialExpressionDistanceToNearestSurface_0}
   ]
  }
  MaterialExpressionMultiply_0 {
   class=MaterialExpressionMultiply
   inputs [
    A {from=MaterialExpressionTextureCoordinate_0}
    B {from=MaterialExpressionAppendVector_1}
   ]
  }
  MaterialExpressionMultiply_1 {
   class=MaterialExpressionMultiply
   inputs [
    A {from=MaterialExpressionAdd_0}
    B {from=MaterialExpressionConstant_5}
   ]
  }
  MaterialExpressionMultiply_2 {
   class=MaterialExpressionMultiply
   inputs [
    A {from=MaterialExpressionParticleColor_0, output=RGB}
    B {from=MaterialExpressionTextureSampleParameter2D_0, output=RGB}
   ]
  }
  MaterialExpressionMultiply_29 {
   class=MaterialExpressionMultiply
   inputs [
    A {from=MaterialExpressionAdd_24}
    B {from=MaterialExpressionReroute_74}
   ]
  }
  MaterialExpressionMultiply_3 {
   class=MaterialExpressionMultiply
   inputs [
    A {from=MaterialExpressionPower_0}
    B {from=MaterialExpressionMultiply_6}
   ]
  }
  MaterialExpressionMultiply_30 {
   class=MaterialExpressionMultiply
   inputs [
    A {from=MaterialExpressionReroute_67}
    B {from=""}
   ]
   properties {constB=-1}
  }
  MaterialExpressionMultiply_31 {
   class=MaterialExpressionMultiply
   inputs [
    A {from=MaterialExpressionReroute_66}
    B {from=MaterialExpressionReroute_68}
   ]
  }
  MaterialExpressionMultiply_4 {
   class=MaterialExpressionMultiply
   inputs [
    A {from=MaterialExpressionTextureSampleParameter2D_2, output=A}
    B {from=MaterialExpressionSmoothStep_0}
   ]
  }
  MaterialExpressionMultiply_5 {
   class=MaterialExpressionMultiply
   inputs [
    A {from=MaterialExpressionMultiply_4}
    B {from=MaterialExpressionParticleColor_1, output=A}
   ]
  }
  MaterialExpressionMultiply_6 {
   class=MaterialExpressionMultiply
   inputs [
    A {from=MaterialExpressionVectorParameter_0, output=RGB}
    B {from=MaterialExpressionScalarParameter_8}
   ]
  }
  MaterialExpressionOneMinus_1 {class=MaterialExpressionOneMinus, inputs [None {from=MaterialExpressionTextureSampleParameter2D_3, output=R}]}
  MaterialExpressionOneMinus_10 {class=MaterialExpressionOneMinus, inputs [None {from=MaterialExpressionReroute_73}]}
  MaterialExpressionOneMinus_9 {class=MaterialExpressionOneMinus, inputs [None {from=MaterialExpressionScalarParameter_26}]}
  MaterialExpressionParticleColor_0 {class=MaterialExpressionParticleColor, inputs [], properties {}}
  MaterialExpressionParticleColor_1 {class=MaterialExpressionParticleColor, inputs [], properties {}}
  MaterialExpressionPower_0 {
   class=MaterialExpressionPower
   inputs [
    Base {from=MaterialExpressionTextureSampleParameter2D_1, output=RGB}
    Exp {from=MaterialExpressionScalarParameter_4}
   ]
  }
  MaterialExpressionReroute_0 {class=MaterialExpressionReroute, inputs [None {from=MaterialExpressionSubtract_1}]}
  MaterialExpressionReroute_64 {class=MaterialExpressionReroute, inputs [None {from=MaterialExpressionWorldPosition_5, output=XYZ}]}
  MaterialExpressionReroute_65 {class=MaterialExpressionReroute, inputs [None {from=MaterialExpressionDistanceToNearestSurface_4}]}
  MaterialExpressionReroute_66 {class=MaterialExpressionReroute, inputs [None {from=MaterialExpressionAbs_7}]}
  MaterialExpressionReroute_67 {class=MaterialExpressionReroute, inputs [None {from=MaterialExpressionSign_4}]}
  MaterialExpressionReroute_68 {class=MaterialExpressionReroute, inputs [None {from=MaterialExpressionMultiply_30}]}
  MaterialExpressionReroute_69 {class=MaterialExpressionReroute, inputs [None {from=MaterialExpressionMultiply_31}]}
  MaterialExpressionReroute_70 {class=MaterialExpressionReroute, inputs [None {from=MaterialExpressionClamp_6}]}
  MaterialExpressionReroute_71 {class=MaterialExpressionReroute, inputs [None {from=MaterialExpressionConvert_9}]}
  MaterialExpressionReroute_72 {class=MaterialExpressionReroute, inputs [None {from=MaterialExpressionConvert_10}]}
  MaterialExpressionReroute_73 {class=MaterialExpressionReroute, inputs [None {from=MaterialExpressionSmoothStep_6}]}
  MaterialExpressionReroute_74 {class=MaterialExpressionReroute, inputs [None {from=MaterialExpressionOneMinus_10}]}
  MaterialExpressionScalarParameter_0 {class=MaterialExpressionScalarParameter, properties {defaultValue=0.01, parameterName="Height Ratio (S)"}}
  MaterialExpressionScalarParameter_1 {class=MaterialExpressionScalarParameter, properties {defaultValue=32, parameterName="Min Steps (S)"}}
  MaterialExpressionScalarParameter_2 {class=MaterialExpressionScalarParameter, properties {defaultValue=64, parameterName="Max Steps (S)"}}
  MaterialExpressionScalarParameter_25 {class=MaterialExpressionScalarParameter, properties {defaultValue=3.2, parameterName=DistanceHoverOffset}}
  MaterialExpressionScalarParameter_26 {class=MaterialExpressionScalarParameter, properties {defaultValue=500, parameterName=MaxProjectionDistance}}
  MaterialExpressionScalarParameter_27 {class=MaterialExpressionScalarParameter, properties {defaultValue=150, parameterName=AirGapStart}}
  MaterialExpressionScalarParameter_28 {class=MaterialExpressionScalarParameter, properties {defaultValue=200, parameterName=AirGapFade}}
  MaterialExpressionScalarParameter_3 {class=MaterialExpressionScalarParameter, properties {parameterName="Reference Plane (S)"}}
  MaterialExpressionScalarParameter_4 {class=MaterialExpressionScalarParameter, properties {defaultValue=1, parameterName=LavaPower}}
  MaterialExpressionScalarParameter_5 {class=MaterialExpressionScalarParameter, properties {defaultValue=0.1, parameterName=Progress_Erosion_offset}}
  MaterialExpressionScalarParameter_6 {class=MaterialExpressionScalarParameter, properties {defaultValue=0.1, parameterName=Falloff_Erosion}}
  MaterialExpressionScalarParameter_7 {class=MaterialExpressionScalarParameter, properties {parameterName=FRONT_END_WPO_MOD}}
  MaterialExpressionScalarParameter_8 {class=MaterialExpressionScalarParameter, properties {defaultValue=1, parameterName=Front-End_Color_Mod}}
  MaterialExpressionScalarParameter_9 {class=MaterialExpressionScalarParameter, properties {parameterName=Z}}
  MaterialExpressionSign_4 {class=MaterialExpressionSign, inputs [None {from=MaterialExpressionAbs_7, output=""}], properties {}}
  MaterialExpressionSmoothStep_0 {
   class=MaterialExpressionSmoothStep
   inputs [
    Min {from=MaterialExpressionMaterialFunctionCall_8, output=Result}
    Max {from=MaterialExpressionAdd_2, output=""}
    Value {from=MaterialExpressionIf_0, output=""}
   ]
   properties {}
  }
  MaterialExpressionSmoothStep_6 {
   class=MaterialExpressionSmoothStep
   inputs [
    Min {from=MaterialExpressionScalarParameter_27, output=""}
    Max {from=MaterialExpressionAdd_25, output=""}
    Value {from=MaterialExpressionReroute_66, output=""}
   ]
   properties {}
  }
  MaterialExpressionSubtract_0 {
   class=MaterialExpressionSubtract
   inputs [
    A {from=""}
    B {from=MaterialExpressionScalarParameter_6}
   ]
   properties {constA=0}
  }
  MaterialExpressionSubtract_1 {
   class=MaterialExpressionSubtract
   inputs [
    A {from=MaterialExpressionWorldPosition_5, output=XYZ}
    B {from=MaterialExpressionConvert_0}
   ]
  }
  MaterialExpressionTextureCoordinate_0 {class=MaterialExpressionTextureCoordinate}
  MaterialExpressionTextureObjectParameter_0 {
   class=MaterialExpressionTextureObjectParameter
   inputs [
    UVs {from="", output=""}
    Apply View MipBias {from="", output=""}
   ]
   properties {parameterName=HeightMap, texture=/BRCosmetics/Animation/Game/MainPlayer/Emotes/CanineCronutDig/FX/T_Emote_CanineCronutDig_Crack_Flip}
  }
  MaterialExpressionTextureSampleParameter2D_0 {
   class=MaterialExpressionTextureSampleParameter2D
   inputs [
    UVs {from=MaterialExpressionMaterialFunctionCall_2, output=Parallax UVs}
    Apply View MipBias {from="", output=""}
   ]
   properties {
    parameterName=Base
    samplerType=SAMPLERTYPE_LinearColor
    texture=/BRCosmetics/Animation/Game/MainPlayer/Emotes/CanineCronutDig/FX/T_Emote_CanineCronutDig_Crack
   }
  }
  MaterialExpressionTextureSampleParameter2D_1 {
   class=MaterialExpressionTextureSampleParameter2D
   inputs [
    UVs {from=MaterialExpressionMaterialFunctionCall_2, output=Parallax UVs}
    Apply View MipBias {from="", output=""}
   ]
   properties {parameterName=Height, texture=&T_Emote_CanineCronutDig_Crack_Flip}
  }
  MaterialExpressionTextureSampleParameter2D_2 {
   class=MaterialExpressionTextureSampleParameter2D
   inputs [
    UVs {from="", output=""}
    Apply View MipBias {from="", output=""}
   ]
   properties {
    parameterName=Base
    samplerType=SAMPLERTYPE_LinearColor
    texture=&T_Emote_CanineCronutDig_Crack
   }
  }
  MaterialExpressionTextureSampleParameter2D_3 {
   class=MaterialExpressionTextureSampleParameter2D
   inputs [
    UVs {from=MaterialExpressionMaterialFunctionCall_2, output=Parallax UVs}
    Apply View MipBias {from="", output=""}
   ]
   properties {
    parameterName=Base
    samplerType=SAMPLERTYPE_LinearColor
    texture=&T_Emote_CanineCronutDig_Crack
   }
  }
  MaterialExpressionVectorParameter_0 {class=MaterialExpressionVectorParameter, properties {parameterName=LavaColor, defaultValue=<rgba 1 0.2614 0 1>}}
  MaterialExpressionViewProperty_0 {class=MaterialExpressionViewProperty, inputs [], properties {property=MEVP_PreExposure}}
  MaterialExpressionWorldPosition_5 {class=MaterialExpressionWorldPosition, inputs [], properties {}}
 }
 properties {
  bAutomaticallySetUsageInEditor=false
  bUsedWithNiagaraMeshParticles=true
  bUsedWithNiagaraSprites=true
  blendMode=BLEND_Translucent
  usedShadingModels=Default Lit
 }
}
