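# Translucent lit Niagara mesh-particle material for the CanineCronutDig emote, tinting a crack texture by particle color and alpha.
# Its world position offset math uses distance-field distance to pull vertices down in Z toward nearby surfaces, clamped and raised by a hover offset.
# A smoothstep fades the offset across large air gaps.
# FRONT_END_WPO_MOD below one (the default) zeroes the offset, as do mobile feature levels.
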Material /BRCosmetics/Animation/Game/MainPlayer/Emotes/CanineCronutDig/FX/HighEnd/M_Fast_Sprite_WPO
graph {
 expressions {
  MaterialExpressionAbs_0 {class=MaterialExpressionAbs, inputs [None {from=MaterialExpressionReroute_1, output=""}], properties {}}
  MaterialExpressionAdd_0 {
   class=MaterialExpressionAdd
   inputs [
    A {from=MaterialExpressionReroute_16}
    B {from=MaterialExpressionReroute_17}
   ]
  }
  MaterialExpressionAdd_6 {
   class=MaterialExpressionAdd
   inputs [
    A {from=MaterialExpressionScalarParameter_15}
    B {from=MaterialExpressionScalarParameter_16}
   ]
  }
  MaterialExpressionClamp_1 {
   class=MaterialExpressionClamp
   inputs [
    None {from=MaterialExpressionReroute_11, output=""}
    Min {from=MaterialExpressionOneMinus_1, output=""}
    Max {from=MaterialExpressionScalarParameter_1, output=""}
   ]
   properties {}
  }
  MaterialExpressionConstant_0 {class=MaterialExpressionConstant}
  MaterialExpressionConstant_2 {class=MaterialExpressionConstant}
  MaterialExpressionConvert_2 {
   class=MaterialExpressionConvert
   inputs [
    None {from=""}
    None {from=""}
    None {from=MaterialExpressionReroute_12}
   ]
   properties {
    nodeName=Make Vector3
    convertInputs [
     {type=Scalar, expressionInput {expression=null, inputName=None, mask=0, maskA=0, maskB=0, maskG=0, maskR=0, outputIndex=0}}
     {type=Scalar, expressionInput {expression=null, inputName=None, mask=0, maskA=0, maskB=0, maskG=0, maskR=0, outputIndex=0}}
     {
      type=Scalar
      expressionInput {
       expression=&self:MaterialExpressionReroute_12
       inputName=None
       mask=0
       maskA=0
       maskB=0
       maskG=0
       maskR=0
       outputIndex=0
      }
     }
    ]
    convertOutputs [{type=Vector3}]
   }
  }
  MaterialExpressionConvert_3 {
   class=MaterialExpressionConvert
   inputs [
    None {from=""}
    None {from=""}
    None {from=MaterialExpressionScalarParameter_0}
   ]
   properties {
    nodeName=Make Vector3
    convertInputs [
     {type=Scalar, expressionInput {expression=null, inputName=None, mask=0, maskA=0, maskB=0, maskG=0, maskR=0, outputIndex=0}}
     {type=Scalar, expressionInput {expression=null, inputName=None, mask=0, maskA=0, maskB=0, maskG=0, maskR=0, outputIndex=0}}
     {
      type=Scalar
      expressionInput {
       expression=&self:MaterialExpressionScalarParameter_0
       inputName=None
       mask=0
       maskA=0
       maskB=0
       maskG=0
       maskR=0
       outputIndex=0
      }
     }
    ]
    convertOutputs [{type=Vector3}]
   }
  }
  MaterialExpressionDistanceToNearestSurface_1 {class=MaterialExpressionDistanceToNearestSurface, inputs [World Position {from=MaterialExpressionReroute_0, output=""}], properties {}}
  MaterialExpressionFeatureLevelSwitch_0 {
   class=MaterialExpressionFeatureLevelSwitch
   inputs [
    Default {from=MaterialExpressionIf_0, output=""}
    ES2 {from=MaterialExpressionConstant_0, output=""}
    ES3_1 {from=MaterialExpressionConstant_0, output=""}
    SM4_REMOVED {from=MaterialExpressionConstant_0, output=""}
    SM5 {from=MaterialExpressionIf_0, output=""}
    SM6 {from=MaterialExpressionIf_0, output=""}
   ]
   properties {}
  }
  MaterialExpressionIf_0 {
   class=MaterialExpressionIf
   inputs [
    A {from=MaterialExpressionScalarParameter_2, output=""}
    B {from="", output=""}
    "A > B" {from=MaterialExpressionMultiply_21, output=""}
    "A == B" {from=MaterialExpressionMultiply_21, output=""}
    "A < B" {from=MaterialExpressionConstant_2, output=""}
   ]
   properties {constB=1}
  }
  MaterialExpressionMultiply_0 {
   class=MaterialExpressionMultiply
   inputs [
    A {from=MaterialExpressionParticleColor_1, output=RGB}
    B {from=MaterialExpressionTextureSampleParameter2D_0, output=RGB}
   ]
  }
  MaterialExpressionMultiply_1 {
   class=MaterialExpressionMultiply
   inputs [
    A {from=MaterialExpressionParticleColor_1, output=A}
    B {from=MaterialExpressionTextureSampleParameter2D_0, output=A}
   ]
  }
  MaterialExpressionMultiply_2 {
   class=MaterialExpressionMultiply
   inputs [
    A {from=MaterialExpressionReroute_6}
    B {from=""}
   ]
   properties {constB=-1}
  }
  MaterialExpressionMultiply_21 {
   class=MaterialExpressionMultiply
   inputs [
    A {from=MaterialExpressionAdd_0}
    B {from=MaterialExpressionReroute_19}
   ]
  }
  MaterialExpressionMultiply_3 {
   class=MaterialExpressionMultiply
   inputs [
    A {from=MaterialExpressionReroute_2}
    B {from=MaterialExpressionReroute_9}
   ]
  }
  MaterialExpressionOneMinus_1 {class=MaterialExpressionOneMinus, inputs [None {from=MaterialExpressionScalarParameter_1}]}
  MaterialExpressionOneMinus_2 {class=MaterialExpressionOneMinus, inputs [None {from=MaterialExpressionReroute_18}]}
  MaterialExpressionParticleColor_1 {class=MaterialExpressionParticleColor, inputs [], properties {}}
  MaterialExpressionReroute_0 {class=MaterialExpressionReroute, inputs [None {from=MaterialExpressionWorldPosition_1, output=XYZ}]}
  MaterialExpressionReroute_1 {class=MaterialExpressionReroute, inputs [None {from=MaterialExpressionDistanceToNearestSurface_1}]}
  MaterialExpressionReroute_11 {class=MaterialExpressionReroute, inputs [None {from=MaterialExpressionMultiply_3}]}
  MaterialExpressionReroute_12 {class=MaterialExpressionReroute, inputs [None {from=MaterialExpressionClamp_1}]}
  MaterialExpressionReroute_16 {class=MaterialExpressionReroute, inputs [None {from=MaterialExpressionConvert_2}]}
  MaterialExpressionReroute_17 {class=MaterialExpressionReroute, inputs [None {from=MaterialExpressionConvert_3}]}
  MaterialExpressionReroute_18 {class=MaterialExpressionReroute, inputs [None {from=MaterialExpressionSmoothStep_1}]}
  MaterialExpressionReroute_19 {class=MaterialExpressionReroute, inputs [None {from=MaterialExpressionOneMinus_2}]}
  MaterialExpressionReroute_2 {class=MaterialExpressionReroute, inputs [None {from=MaterialExpressionAbs_0}]}
  MaterialExpressionReroute_6 {class=MaterialExpressionReroute, inputs [None {from=MaterialExpressionSign_0}]}
  MaterialExpressionReroute_9 {class=MaterialExpressionReroute, inputs [None {from=MaterialExpressionMultiply_2}]}
  MaterialExpressionScalarParameter_0 {class=MaterialExpressionScalarParameter, properties {defaultValue=3.2, parameterName=DistanceHoverOffset}}
  MaterialExpressionScalarParameter_1 {class=MaterialExpressionScalarParameter, properties {defaultValue=500, parameterName=MaxProjectionDistance}}
  MaterialExpressionScalarParameter_15 {class=MaterialExpressionScalarParameter, properties {defaultValue=150, parameterName=AirGapStart}}
  MaterialExpressionScalarParameter_16 {class=MaterialExpressionScalarParameter, properties {defaultValue=200, parameterName=AirGapFade}}
  MaterialExpressionScalarParameter_2 {class=MaterialExpressionScalarParameter, properties {parameterName=FRONT_END_WPO_MOD}}
  MaterialExpressionSign_0 {class=MaterialExpressionSign, inputs [None {from=MaterialExpressionAbs_0, output=""}], properties {}}
  MaterialExpressionSmoothStep_1 {
   class=MaterialExpressionSmoothStep
   inputs [
    Min {from=MaterialExpressionScalarParameter_15, output=""}
    Max {from=MaterialExpressionAdd_6, output=""}
    Value {from=MaterialExpressionReroute_2, output=""}
   ]
   properties {}
  }
  MaterialExpressionTextureSampleParameter2D_0 {
   class=MaterialExpressionTextureSampleParameter2D
   inputs [
    UVs {from="", output=""}
    Apply View MipBias {from="", output=""}
   ]
   properties {parameterName=Texture, texture=/BRCosmetics/Animation/Game/MainPlayer/Emotes/CanineCronutDig/FX/T_Emote_CanineCronut_Dig_Crack_Dark}
  }
  MaterialExpressionWorldPosition_1 {class=MaterialExpressionWorldPosition, inputs [], properties {}}
 }
 properties {bAutomaticallySetUsageInEditor=false, bUsedWithNiagaraMeshParticles=true, blendMode=BLEND_Translucent, usedShadingModels=Default Lit}
}
